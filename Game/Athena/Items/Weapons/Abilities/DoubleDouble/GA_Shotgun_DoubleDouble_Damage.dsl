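; Damage gameplay ability for the DoubleDouble shotgun that adds a self-knock effect to the shooter when it fires.
; On commit execute, it applies the SelfKnock gameplay effect to the owner.
; It also applies that effect to the avatar's attach parent when that is an AI pawn with a valid ability system component, covering shots fired while riding.
; The parent class's commit logic runs last.
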(event Ability|EventCommitExecute
  (Utilities|FlowControl|Sequence
    (:then_0
      ;[1 1 1] Apply Knock Effect to self when firing
      (Ability|ApplyGameplayEffectToOwner (Variables|Default|GetGE_SelfKnock)))
    (:then_1
      ;[1 1 1] Apply Knock efrect to the Attached Parent AIPawn Actor (Riding)
      (bind _asfort_aipawn (Utilities|Casting|CastToFortAIPawn (Actor|GetAttachParentActor (Ability|GetAvatarActorfromActorInfo))))
      (Utilities|IsValid (Ability|GetAbilitySystemComponent _asfort_aipawn)
        (:Is%20Valid
          (GameplayEffects|ApplyGameplayEffectToTarget :self (Ability|GetAbilitySystemComponent _asfort_aipawn) :GameplayEffectClass (Variables|Default|GetGE_SelfKnock) :Target (Ability|GetAbilitySystemComponent _asfort_aipawn) :Level 1.0))))
    (:then_2
      (|CallParentFunction))))
