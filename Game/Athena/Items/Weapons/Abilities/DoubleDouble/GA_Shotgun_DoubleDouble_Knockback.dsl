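; Gameplay ability for the DoubleDouble shotgun that knocks the hit pawn away from the shooter, found through the granting effect's causer.
; KnockPlayer skips knockback-immune or in-vehicle pawns, scales launch linearly down to zero at max distance, and adjusts Z using tunable scalable floats.
; It also sets the fall instigator and optionally resets fall height on upward launches.
; A delayed velocity check plays a distance-scaled GameplayCue if vertical speed is high, then ends the ability.
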(event Ability|EventActivateAbility
  (CallFunction|KnockPlayer)
  ;[1 1 1] Wait a brief moment (Like 2 frames) then check velocity. If Velocity on Z is high enough or low enough play GC.
  (Utilities|Time|SetTimerbyEvent (delegate Custom|CheckVel) (Variables|Default|GetFX%20Delay)))

(event Ability|EventFailedToActivatePassiveAbility
  (params (ActorInfo :type Gameplay%20Ability%20Actor%20Info%20Structure))
  (|CallParentFunction))

(event Custom|CheckVel
  (bind _returnvalue (Ability|GetAvatarActorfromActorInfo))
  ;[1 1 1] Wait a brief moment (Like 2 frames) then check velocity. If Velocity on Z is high enough or low enough play GC.
  (Utilities|IsValid _returnvalue
    (:Is%20Valid
      (bind _asfort_player_pawn (Utilities|Casting|CastToFortPlayerPawn _returnvalue))
      (Development|PrintString (Utilities|String|ToString(Vector) (Transformation|GetVelocity _asfort_player_pawn)))
      (bind (_returnvalue_x _returnvalue_y _returnvalue_z) (Math|Vector|BreakVector (Transformation|GetVelocity _asfort_player_pawn)))
      (if (> (Math|Float|Absolute(Float) _returnvalue_z) (Variables|Default|GetVelocityForFX))
        ;;[1 1 1] Play Cue with intesity
        (Ability|ExecuteGameplayCueWithParamsOnOwner (Variables|Default|GetKnockbackCue) (Ability|GameplayCue|MakeGameplayCueParameters :NormalizedMagnitude (Variables|Default|GetDistanceMagMod) :RawMagnitude 0.0 :EffectContext "()" :AggregatedSourceTags "(GameplayTags=)" :AggregatedTargetTags "(GameplayTags=)" :Location "0,0,0" :Normal "0,0,0" :GameplayEffectLevel 1 :AbilityLevel 1 :bReplicateLocationWhenUsingMinimalRepProxy false))
        (else
          _))
      (Ability|EndAbility))))

(fn KnockPlayer
  (bind _row_maxdistance (Variables|Default|GetRow_MaxDistance))
  (bind _returnvalue (ScalableFloat|GetValueatLevel _row_maxdistance))
  (bind _returnvalue_2 (Ability|GetGrantedbyEffectContext))
  (bind _returnvalue_3 (Ability|EffectContext|GetEffectCauser _returnvalue_2))
  (bind _returnvalue_4 (Actor|GetOwner _returnvalue_3))
  (bind _returnvalue_5 (Transformation|GetActorLocation _returnvalue_4))
  (bind _returnvalue_6 (Ability|GetAvatarActorfromActorInfo))
  (bind _finallaunchvel (Variables|Default|GetFinalLaunchVel))
  ;[1 1 1] Cast to pawn and launch the player a little bit. The distance launched is based on a max distance from the player who hit them. If the player is further than the max distance than no launch is applied. Otherwise the player is launched based on a ratio of the knock mag using the distance from the player who launched them.
  (Utilities|IsValid _returnvalue_6
    (:Is%20Valid
      ;;[1 1 1] Cast to pawn and check immunity/invehicle
      (bind _asfort_pawn (Utilities|Casting|CastToFortPawn _returnvalue_6))
      (if (not (or (GameplayTags|HasMatchingGameplayTag _asfort_pawn "(TagName=\"Gameplay.Status.KnockbackImmunity\")") (GameplayTags|HasMatchingGameplayTag _asfort_pawn "(TagName=\"Gameplay.Action.Player.AthenaInVehicle\")")))
        (Variables|Default|SetLauncher _returnvalue_4)
        ;;[1 1 1] Calc Distance to apply knockback harder if closer
        (bind _distance (Class|MissionBlueprintFunctionLibrary|DistanceBetweenTwoVectors (Transformation|GetActorLocation _asfort_pawn) _returnvalue_5))
        (Variables|Default|SetDistance _distance)
        (if (< (Variables|Default|GetDistance) _returnvalue)
          (Variables|Default|SetDistanceMagMod (Math|Float|Absolute(Float) (- (/ (Variables|Default|GetDistance) _returnvalue) 1.0)))
          ;;[1 1 1] Calc Launch Vel
          (bind (_returnvalue_x _returnvalue_y _returnvalue_z) (Math|Vector|BreakVector (* (- _returnvalue_5 (Transformation|GetActorLocation _asfort_pawn)) -1.0)))
          (Variables|Default|SetFinalLaunchVel (Math|Vector|MakeVector (.x (* (Math|Vector|Normalize (Math|Vector|MakeVector _returnvalue_x _returnvalue_y (+ _returnvalue_z (ScalableFloat|GetValueatLevel (Variables|Default|GetRow_ZStartOffset))))) (* (ScalableFloat|GetValueatLevel (Variables|Default|GetRow_KnockMag)) (Variables|Default|GetDistanceMagMod)))) (.y (* (Math|Vector|Normalize (Math|Vector|MakeVector _returnvalue_x _returnvalue_y (+ _returnvalue_z (ScalableFloat|GetValueatLevel (Variables|Default|GetRow_ZStartOffset))))) (* (ScalableFloat|GetValueatLevel (Variables|Default|GetRow_KnockMag)) (Variables|Default|GetDistanceMagMod)))) (Math|Float|float+float (.z (* (Math|Vector|Normalize (Math|Vector|MakeVector _returnvalue_x _returnvalue_y (+ _returnvalue_z (ScalableFloat|GetValueatLevel (Variables|Default|GetRow_ZStartOffset))))) (* (ScalableFloat|GetValueatLevel (Variables|Default|GetRow_KnockMag)) (Variables|Default|GetDistanceMagMod)))) (ScalableFloat|GetValueatLevel (Variables|Default|GetRow_ZAdd)) (* (.z (Transformation|GetVelocity _asfort_pawn)) (ScalableFloat|GetValueatLevel (Variables|Default|GetRow_ZVelMult))))))
          ;;[1 1 1] Stop Movement and launch them
          (Components|Movement|StopMovementImmediately (Pawn|GetMovementComponent _asfort_pawn))
          (Character|LaunchCharacter _asfort_pawn _finallaunchvel true true)
          ;;[1 1 1] Set Fall Instigator
          (bind _asfort_player_pawn_athena (Utilities|Casting|CastToFortPlayerPawnAthena _asfort_pawn)
            (:then
              (Class|FortPlayerPawnAthena|SetFallInstigator _asfort_player_pawn_athena (Fortnite|GetFortPlayerControllerfromActor (Variables|Default|GetLauncher)) (Variables|Default|GetDefaultTags)))
            (:CastFailed
              _))
          ;;[1 1 1] Check if we went pos z, if so reset fall height
          (if (Math|Conversions|ToBoolean(Integer) (Math|Float|Floor (ScalableFloat|GetValueatLevel (Variables|Default|GetRow_ResetFall))))
            (if (> (.z _finallaunchvel) 0.0)
              (Pawn|Character|ResetFallingHeight _asfort_pawn))))))))
